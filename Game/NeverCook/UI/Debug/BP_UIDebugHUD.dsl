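; BP_UIDebugHUD is a debug HUD Blueprint that displays a test UI widget when play begins.
; On BeginPlay it creates the widget from the class in the TestWidget variable and adds it to the viewport.
; It then puts the owning player controller into UI-only input mode and shows the mouse cursor.
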(event EventBeginPlay
  (bind _returnvalue (HUD|GetOwningPlayerController))
  (bind _returnvalue_2 (UserInterface|CreateWidget (Variables|Default|GetTestWidget)))
  (UserInterface|Viewport|AddtoViewport _returnvalue_2 0)
  (Input|SetInputModeUIOnly _returnvalue)
  (Variables|/Script/Engine.PlayerController|SetbShowMouseCursor true _returnvalue))

(fn ConstructionScript)
